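# Widget Blueprint for one row of the Athena replay spectator action feed, showing an event time, instigator button, event icon (broadcast-eliminations art by default) and victim button.
# It derives from AthenaReplayActionFeedRowBase, which supplies those four widgets, and lays them out in a bordered horizontal row with fixed-size name buttons.
# The graph only handles the name-updated events, casting each inherited button to IconTextButton to set the instigator or victim text.
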WidgetBlueprint /FortniteCoreUI/Athena/UI/Spectator/ActionFeed/AthenaReplayActionFeedRow
graphs [EventGraph {nodes [
   n1 = Event OnInstigatorNameUpdated!() -> (OutputDelegate, then, InstigatorName:"Text (by ref)") {is=title K2Node_Event}
   > n2 = Cast To IconTextButton!(execute, Object=Get Button_InstigatorButton() -> (Button_InstigatorButton:Common Button Legacy Object Reference)) -> (then, CastFailed, AsIcon Text Button:Icon Text Button Object Reference) {is=title K2Node_DynamicCast}
   > Set Text!(execute, self:Icon Text Button Object Reference=@n2, ButtonText:Text=@n1.InstigatorName) -> (then) {is=title K2Node_CallFunction}
   n3 = Event OnVictimNameUpdated!() -> (OutputDelegate, then, VictimName:"Text (by ref)") {is=title K2Node_Event}
   > n4 = Cast To IconTextButton(Object=Get Button_VictimButton() -> (Button_VictimButton:Common Button Legacy Object Reference))
   > Set Text(self=@n4, ButtonText=@n3.VictimName)
  ]}]
variables []
widgets {
 info {
  inheritedWidgetCount=4
  parentClass=/Script/FortniteUI.AthenaReplayActionFeedRowBase
  rootWidgetClass=&CommonBorder
  widgetCount=10
 }
 widgets [
  CommonBorder_0 {
   widgetClassPath=&CommonBorder
   widget=CommonBorder_0
   properties {
    horizontalAlignment=HAlign_Center
    style=/FortUILibrary/Styles/Border/Border_AthenaStore_OfferHeaderBacking.Border_AthenaStore_OfferHeaderBacking_C
    verticalAlignment=VAlign_Center
    brushColor=<rgba 1 1 1 0.5>
    padding=<ltrb 10 0 10 0>
   }
  }
  HorizontalBox_1 {
   widgetClassPath=&HorizontalBox
   parent=CommonBorder_0
   widget=HorizontalBox_1
   slot {class=BorderSlot, properties {horizontalAlignment=HAlign_Center, verticalAlignment=VAlign_Center, padding=<ltrb 10 0 10 0>}}
  }
  SizeBox_2 {
   widgetClassPath=&SizeBox
   parent=HorizontalBox_1
   widget=SizeBox_2
   properties {bOverride_MinDesiredWidth=true, minDesiredWidth=67, widthOverride=70}
   slot {class=HorizontalBoxSlot, properties {verticalAlignment=VAlign_Center, padding=<ltrb 0 0 10 0>}}
  }
  Text_EventTime {
   widgetClassPath=&CommonTextBlock
   bInherited=true
   bIsVariable=true
   parent=SizeBox_2
   widget=Text_EventTime
   properties {
    justification=Center
    text=88:88
    font {
     bForceMonospaced=false
     bMaterialIsStencil=false
     fontMaterial=null
     fontObject=&NotoSans
     letterSpacing=0
     monospacedWidth=1
     size=20
     skewAmount=0
     typefaceFontName=NotoSans-Regular
     outlineSettings {
      bApplyOutlineToDropShadows=false
      bMiteredCorners=false
      bSeparateFillAlpha=false
      outlineMaterial=null
      outlineSize=0
      outlineColor=<rgba 0 0 0 1>
     }
    }
    shadowOffset=<0 0>
   }
   slot {class=SizeBoxSlot, properties {horizontalAlignment=HAlign_Right, verticalAlignment=VAlign_Center}}
  }
  ButtonsHB {
   widgetClassPath=&HorizontalBox
   parent=HorizontalBox_1
   widget=ButtonsHB
   slot {class=HorizontalBoxSlot, properties {horizontalAlignment=HAlign_Center, size {sizeRule=Fill, value=1}}}
  }
  InstigatorSB {
   widgetClassPath=&SizeBox
   parent=ButtonsHB
   widget=InstigatorSB
   properties {bOverride_HeightOverride=true, bOverride_WidthOverride=true, heightOverride=40, widthOverride=225}
   slot {class=HorizontalBoxSlot, properties {verticalAlignment=VAlign_Center}}
  }
  Button_InstigatorButton {
   widgetClassPath=&IconTextButton
   bInherited=true
   bIsVariable=true
   parent=InstigatorSB
   widget=Button_InstigatorButton
   properties {
    buttonText=?????
    contentAlignment=HAlign_Center
    onButtonBaseClicked=(&self:WidgetTree.Button_InstigatorButton.HandleOnButtonClicked)
    onButtonBaseDoubleClicked=(&self:WidgetTree.Button_InstigatorButton.HandleOnButtonDoubleClicked)
    onButtonBaseHovered=(&self:WidgetTree.Button_InstigatorButton.HandleOnButtonHovered)
    onButtonBaseUnhovered=(&self:WidgetTree.Button_InstigatorButton.HandleOnButtonUnhovered)
    onSelectedChangedBase=(&self:WidgetTree.Button_InstigatorButton.HandleOnSelectedChanged)
    style=/FortUILibrary/Foundation/Buttons/ButtonStyles/TopBar/ButtonStyle-Outline-S.ButtonStyle-Outline-S_C
   }
   slot {class=SizeBoxSlot}
  }
  Image_EventIcon {
   widgetClassPath=&Image
   bInherited=true
   bIsVariable=true
   parent=ButtonsHB
   widget=Image_EventIcon
   properties {
    visibility=SelfHitTestInvisible
    brush {
     bIsDynamicallyLoaded=false
     drawAs=Image
     imageType=NoImage
     mirroring=NoMirror
     resourceName=None
     resourceObject=/Game/Athena/HUD/Spectator/Art/T-Icon-BroadcastEliminations
     tiling=NoTile
     imageSize=<40 40>
     margin=<ltrb 0 0 0 0>
     outlineSettings {
      bUseBrushTransparency=false
      roundingType=HalfHeightRadius
      width=0
      color {colorUseRule=UseColor_Specified, specifiedColor=<rgba 0 0 0 0>}
      cornerRadii=<0 0 0 0>
     }
     tintColor {colorUseRule=UseColor_Specified, specifiedColor=<rgba 1 1 1 1>}
     uVRegion {bIsValid=false, max=<0 0>, min=<0 0>}
    }
   }
   slot {class=HorizontalBoxSlot, properties {horizontalAlignment=HAlign_Center, verticalAlignment=VAlign_Center, padding=<ltrb 5 0 5 0>}}
  }
  VictimSB {
   widgetClassPath=&SizeBox
   parent=ButtonsHB
   widget=VictimSB
   properties {bOverride_HeightOverride=true, bOverride_WidthOverride=true, heightOverride=40, widthOverride=225}
   slot {class=HorizontalBoxSlot, properties {verticalAlignment=VAlign_Center}}
  }
  Button_VictimButton {
   widgetClassPath=&IconTextButton
   bInherited=true
   bIsVariable=true
   parent=VictimSB
   widget=Button_VictimButton
   properties {
    buttonText=??????
    contentAlignment=HAlign_Center
    onButtonBaseClicked=(&self:WidgetTree.Button_VictimButton.HandleOnButtonClicked)
    onButtonBaseDoubleClicked=(&self:WidgetTree.Button_VictimButton.HandleOnButtonDoubleClicked)
    onButtonBaseHovered=(&self:WidgetTree.Button_VictimButton.HandleOnButtonHovered)
    onButtonBaseUnhovered=(&self:WidgetTree.Button_VictimButton.HandleOnButtonUnhovered)
    onSelectedChangedBase=(&self:WidgetTree.Button_VictimButton.HandleOnSelectedChanged)
    style=&ButtonStyle-Outline-S_C
   }
   slot {class=SizeBoxSlot}
  }
 ]
}
